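# Material function for the RankedScythe female body that stamps a rank icon into the material's normals and TechArt mask.
# RankIndex selects a cell from 3x3 icon textures, with UVs scaled, offset, rotated and mirrored across a U threshold, including normal correction.
# A box-masked icon mask drives the attribute blend and adds to or subtracts from each TechArt channel.
# Static switches bypass the icon or enable a box-mask debug view.
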MaterialFunction /BRCosmetics/Characters/Player/Female/Medium/Bodies/F_MED_RankedScythe/Materials/MF_RankedScytheIcon
graph {expressions {
  MaterialExpressionAdd_0 {
   class=MaterialExpressionAdd
   inputs [
    A {from=MaterialExpressionMaterialFunctionCall_2, output=UVs}
    B {from=MaterialExpressionMaterialFunctionCall_3, output=Result}
   ]
  }
  MaterialExpressionAdd_1 {
   class=MaterialExpressionAdd
   inputs [
    A {from=MaterialExpressionNamedRerouteUsage_4}
    B {from=MaterialExpressionAdd_4}
   ]
  }
  MaterialExpressionAdd_3 {
   class=MaterialExpressionAdd
   inputs [
    A {from=MaterialExpressionConvert_2}
    B {from=MaterialExpressionNamedRerouteUsage_11}
   ]
  }
  MaterialExpressionAdd_4 {
   class=MaterialExpressionAdd
   inputs [
    A {from=MaterialExpressionMaterialFunctionCall_4, output=Rotated Values}
    B {from=MaterialExpressionMultiply_1}
   ]
  }
  MaterialExpressionAdd_7 {
   class=MaterialExpressionAdd
   inputs [
    A {from=MaterialExpressionConvert_2}
    B {from=MaterialExpressionNamedRerouteUsage_11}
   ]
  }
  MaterialExpressionAdd_8 {
   class=MaterialExpressionAdd
   inputs [
    A {from=MaterialExpressionNamedRerouteDeclaration_8}
    B {from=MaterialExpressionNamedRerouteUsage_11}
   ]
  }
  MaterialExpressionAppendVector_0 {
   class=MaterialExpressionAppendVector
   inputs [
    A {from=MaterialExpressionScalarParameter_2, output=""}
    B {from=MaterialExpressionScalarParameter_5, output=""}
   ]
  }
  MaterialExpressionAppendVector_1 {
   class=MaterialExpressionAppendVector
   inputs [
    A {from=MaterialExpressionScalarParameter_8, output=""}
    B {from=MaterialExpressionScalarParameter_9, output=""}
   ]
  }
  MaterialExpressionAppendVector_4 {
   class=MaterialExpressionAppendVector
   inputs [
    A {from=MaterialExpressionConstant_1, output=""}
    B {from=MaterialExpressionConstant_1, output=""}
   ]
  }
  MaterialExpressionBlendMaterialAttributes_1 {
   class=MaterialExpressionBlendMaterialAttributes
   inputs [
    A {from=MaterialExpressionGetMaterialAttributes_1, output=MaterialAttributes}
    B {from=MaterialExpressionSetMaterialAttributes_1, output=""}
    Alpha {from=MaterialExpressionNamedRerouteUsage_0, output=""}
   ]
   properties {vertexAttributeBlendType=UseA}
  }
  MaterialExpressionConstant2Vector_0 {class=MaterialExpressionConstant2Vector, properties {g=1, r=-1}}
  MaterialExpressionConstant3Vector_0 {class=MaterialExpressionConstant3Vector, properties {constant=<rgba 0 0 1 0>}}
  MaterialExpressionConstant3Vector_1 {class=MaterialExpressionConstant3Vector, properties {constant=<rgba -1 1 1 0>}}
  MaterialExpressionConstant_0 {class=MaterialExpressionConstant}
  MaterialExpressionConstant_1 {class=MaterialExpressionConstant, properties {r=0.175}}
  MaterialExpressionConstant_14 {class=MaterialExpressionConstant, properties {r=3}}
  MaterialExpressionConstant_5 {class=MaterialExpressionConstant, properties {r=0.01}}
  MaterialExpressionConstant_6 {class=MaterialExpressionConstant, properties {r=3}}
  MaterialExpressionConvert_0 {
   class=MaterialExpressionConvert
   inputs [
    None {from=MaterialExpressionLinearInterpolate_5}
    None {from=MaterialExpressionLinearInterpolate_4}
    None {from=MaterialExpressionLinearInterpolate_3}
   ]
   properties {
    nodeName=Make Float3
    convertInputs [
     {
      type=Scalar
      expressionInput {
       expression=&self:MaterialExpressionLinearInterpolate_5
       inputName=None
       mask=0
       maskA=0
       maskB=0
       maskG=0
       maskR=0
       outputIndex=0
      }
     }
     {
      type=Scalar
      expressionInput {
       expression=&self:MaterialExpressionLinearInterpolate_4
       inputName=None
       mask=0
       maskA=0
       maskB=0
       maskG=0
       maskR=0
       outputIndex=0
      }
     }
     {
      type=Scalar
      expressionInput {
       expression=&self:MaterialExpressionLinearInterpolate_3
       inputName=None
       mask=0
       maskA=0
       maskB=0
       maskG=0
       maskR=0
       outputIndex=0
      }
     }
    ]
    convertOutputs [{type=Vector3}]
   }
  }
  MaterialExpressionConvert_2 {
   class=MaterialExpressionConvert
   inputs [None {from=MaterialExpressionFunctionInput_2}]
   properties {
    nodeName=Break Float3
    convertInputs [{
      type=Vector3
      expressionInput {
       expression=&self:MaterialExpressionFunctionInput_2
       inputName=None
       mask=0
       maskA=0
       maskB=0
       maskG=0
       maskR=0
       outputIndex=0
      }
     }]
    convertOutputs [{type=Scalar}, {type=Scalar}, {type=Scalar}]
   }
  }
  MaterialExpressionDivide_2 {
   class=MaterialExpressionDivide
   inputs [
    A {from=MaterialExpressionScalarParameter_18}
    B {from=MaterialExpressionConstant_14}
   ]
  }
  MaterialExpressionDivide_4 {
   class=MaterialExpressionDivide
   inputs [
    A {from=""}
    B {from=MaterialExpressionConstant_14}
   ]
  }
  MaterialExpressionDivide_5 {
   class=MaterialExpressionDivide
   inputs [
    A {from=MaterialExpressionFmod_0}
    B {from=MaterialExpressionConstant_6}
   ]
  }
  MaterialExpressionFloor_0 {class=MaterialExpressionFloor, inputs [None {from=MaterialExpressionDivide_2, output=""}], properties {}}
  MaterialExpressionFmod_0 {
   class=MaterialExpressionFmod
   inputs [
    A {from=MaterialExpressionScalarParameter_18, output=""}
    B {from=MaterialExpressionConstant_6, output=""}
   ]
   properties {}
  }
  MaterialExpressionFunctionInput_1 {
   class=MaterialExpressionFunctionInput
   inputs [Preview {from="", output=""}]
   properties {inputName=Material, inputType=FunctionInput_MaterialAttributes}
  }
  MaterialExpressionFunctionInput_2 {class=MaterialExpressionFunctionInput, inputs [Preview {from="", output=""}], properties {inputName=TechArt Mask, sortPriority=1}}
  MaterialExpressionFunctionOutput_0 {class=MaterialExpressionFunctionOutput, inputs [None {from=MaterialExpressionStaticSwitch_0, output=""}]}
  MaterialExpressionFunctionOutput_1 {
   class=MaterialExpressionFunctionOutput
   inputs [None {from=MaterialExpressionSaturate_4, output=""}]
   properties {outputName=TechArt Mask, sortPriority=1}
  }
  MaterialExpressionGetMaterialAttributes_1 {
   class=MaterialExpressionGetMaterialAttributes
   inputs [None {from=MaterialExpressionFunctionInput_1, output=""}]
   properties {attributeGetTypes [#1]}
  }
  MaterialExpressionLinearInterpolate_0 {
   class=MaterialExpressionLinearInterpolate
   inputs [
    A {from=MaterialExpressionConstant3Vector_0}
    B {from=MaterialExpressionMaterialFunctionCall_1, output=Result}
    Alpha {from=MaterialExpressionNamedRerouteUsage_3}
   ]
  }
  MaterialExpressionLinearInterpolate_1 {
   class=MaterialExpressionLinearInterpolate
   inputs [
    A {from=MaterialExpressionAdd_0}
    B {from=MaterialExpressionMultiply_0}
    Alpha {from=MaterialExpressionNamedRerouteDeclaration_7}
   ]
  }
  MaterialExpressionLinearInterpolate_2 {
   class=MaterialExpressionLinearInterpolate
   inputs [
    A {from=MaterialExpressionMultiply_2}
    B {from=MaterialExpressionTextureSample_0, output=RGB}
    Alpha {from=MaterialExpressionNamedRerouteUsage_12}
   ]
  }
  MaterialExpressionLinearInterpolate_3 {
   class=MaterialExpressionLinearInterpolate
   inputs [
    A {from=MaterialExpressionAdd_3}
    B {from=MaterialExpressionSubtract_2}
    Alpha {from=MaterialExpressionScalarParameter_12}
   ]
  }
  MaterialExpressionLinearInterpolate_4 {
   class=MaterialExpressionLinearInterpolate
   inputs [
    A {from=MaterialExpressionAdd_7}
    B {from=MaterialExpressionSubtract_3}
    Alpha {from=MaterialExpressionScalarParameter_13}
   ]
  }
  MaterialExpressionLinearInterpolate_5 {
   class=MaterialExpressionLinearInterpolate
   inputs [
    A {from=MaterialExpressionAdd_8}
    B {from=MaterialExpressionSubtract_5}
    Alpha {from=MaterialExpressionScalarParameter_14}
   ]
  }
  MaterialExpressionMaterialFunctionCall_0 {
   class=MaterialExpressionMaterialFunctionCall
   inputs [
    BaseNormal {from=MaterialExpressionNamedRerouteUsage_1}
    AdditionalNormal {from=MaterialExpressionLinearInterpolate_0}
   ]
   properties {materialFunction=&BlendAngleCorrectedNormals}
  }
  MaterialExpressionMaterialFunctionCall_1 {
   class=MaterialExpressionMaterialFunctionCall
   inputs [
    Normal {from=MaterialExpressionLinearInterpolate_2}
    Flatness {from=MaterialExpressionScalarParameter_0}
   ]
   properties {materialFunction=&FlattenNormal}
  }
  MaterialExpressionMaterialFunctionCall_12 {
   class=MaterialExpressionMaterialFunctionCall
   inputs [
    X {from=MaterialExpressionDivide_5}
    Y {from=MaterialExpressionMultiply_8}
   ]
   properties {materialFunction=&MakeFloat2}
  }
  MaterialExpressionMaterialFunctionCall_2 {
   class=MaterialExpressionMaterialFunctionCall
   inputs [
    UVs {from=MaterialExpressionTextureCoordinate_0}
    Texture Scale {from=MaterialExpressionScalarParameter_1}
   ]
   properties {materialFunction=&ScaleUVsByCenter}
  }
  MaterialExpressionMaterialFunctionCall_3 {
   class=MaterialExpressionMaterialFunctionCall
   inputs [
    X {from=MaterialExpressionScalarParameter_3}
    Y {from=MaterialExpressionScalarParameter_4}
   ]
   properties {materialFunction=&MakeFloat2}
  }
  MaterialExpressionMaterialFunctionCall_4 {
   class=MaterialExpressionMaterialFunctionCall
   inputs [
    UVs {from=MaterialExpressionLinearInterpolate_1}
    Rotation Center {from=MaterialExpressionAppendVector_4}
    "Rotation Angle (0-1)" {from=MaterialExpressionScalarParameter_6}
   ]
   properties {materialFunction=&CustomRotator}
  }
  MaterialExpressionMaterialFunctionCall_5 {
   class=MaterialExpressionMaterialFunctionCall
   inputs [
    A {from=MaterialExpressionTextureCoordinate_1}
    B {from=MaterialExpressionAppendVector_0}
    Bounds {from=MaterialExpressionAppendVector_1}
    Edge Falloff {from=MaterialExpressionConstant_5}
   ]
   properties {materialFunction=/Engine/Functions/Engine_MaterialFunctions01/Gradient/BoxMask-2D}
  }
  MaterialExpressionMaterialFunctionCall_6 {
   class=MaterialExpressionMaterialFunctionCall
   inputs [UV Channel {from=""}]
   properties {materialFunction=&LinearGradient}
  }
  MaterialExpressionMaterialFunctionCall_7 {
   class=MaterialExpressionMaterialFunctionCall
   inputs [
    X {from=MaterialExpressionScalarParameter_10}
    Y {from=MaterialExpressionScalarParameter_11}
   ]
   properties {materialFunction=&MakeFloat2}
  }
  MaterialExpressionMultiply_0 {
   class=MaterialExpressionMultiply
   inputs [
    A {from=MaterialExpressionAdd_0}
    B {from=MaterialExpressionConstant2Vector_0}
   ]
  }
  MaterialExpressionMultiply_1 {
   class=MaterialExpressionMultiply
   inputs [
    A {from=MaterialExpressionMaterialFunctionCall_7, output=Result}
    B {from=MaterialExpressionNamedRerouteDeclaration_7}
   ]
  }
  MaterialExpressionMultiply_2 {
   class=MaterialExpressionMultiply
   inputs [
    A {from=MaterialExpressionTextureSample_0, output=RGB}
    B {from=MaterialExpressionConstant3Vector_1}
   ]
  }
  MaterialExpressionMultiply_6 {
   class=MaterialExpressionMultiply
   inputs [
    A {from=MaterialExpressionTextureSample_2, output=R}
    B {from=MaterialExpressionNamedRerouteUsage_3}
   ]
  }
  MaterialExpressionMultiply_8 {
   class=MaterialExpressionMultiply
   inputs [
    A {from=MaterialExpressionDivide_4}
    B {from=MaterialExpressionFloor_0}
   ]
  }
  MaterialExpressionNamedRerouteDeclaration_0 {
   class=MaterialExpressionNamedRerouteDeclaration
   inputs [None {from=MaterialExpressionStaticSwitch_1}]
   properties {name=IconMask, nodeColor=<rgba 0.5469 0 1 1>}
  }
  MaterialExpressionNamedRerouteDeclaration_1 {
   class=MaterialExpressionNamedRerouteDeclaration
   inputs [None {from=MaterialExpressionGetMaterialAttributes_1, output=Normal}]
   properties {name=IN_Normal, nodeColor=<rgba 1 0.4688 0 1>}
  }
  MaterialExpressionNamedRerouteDeclaration_2 {
   class=MaterialExpressionNamedRerouteDeclaration
   inputs [None {from=MaterialExpressionMaterialFunctionCall_0, output=Result}]
   properties {name=OUT_Normal, nodeColor=<rgba 0.125 1 0 1>}
  }
  MaterialExpressionNamedRerouteDeclaration_3 {
   class=MaterialExpressionNamedRerouteDeclaration
   inputs [None {from=MaterialExpressionSaturate_0}]
   properties {name=IconBoxMask, nodeColor=<rgba 0.9219 1 0 1>}
  }
  MaterialExpressionNamedRerouteDeclaration_4 {
   class=MaterialExpressionNamedRerouteDeclaration
   inputs [None {from=MaterialExpressionMaterialFunctionCall_12, output=Result}]
   properties {name=RankOffset, nodeColor=<rgba 0.0312 0 1 1>}
  }
  MaterialExpressionNamedRerouteDeclaration_5 {
   class=MaterialExpressionNamedRerouteDeclaration
   inputs [None {from=MaterialExpressionAdd_1}]
   properties {name=UVs, nodeColor=<rgba 0.4297 0 1 1>}
  }
  MaterialExpressionNamedRerouteDeclaration_7 {
   class=MaterialExpressionNamedRerouteDeclaration
   inputs [None {from=MaterialExpressionStep_1}]
   properties {name=MirrorMask, nodeColor=<rgba 0.4766 0 1 1>}
  }
  MaterialExpressionNamedRerouteDeclaration_8 {
   class=MaterialExpressionNamedRerouteDeclaration
   inputs [None {from=MaterialExpressionConvert_2}]
   properties {name=TechArtIconMask, nodeColor=<rgba 0 0.3672 1 1>}
  }
  MaterialExpressionNamedRerouteUsage_0 {class=MaterialExpressionNamedRerouteUsage}
  MaterialExpressionNamedRerouteUsage_1 {class=MaterialExpressionNamedRerouteUsage}
  MaterialExpressionNamedRerouteUsage_11 {class=MaterialExpressionNamedRerouteUsage}
  MaterialExpressionNamedRerouteUsage_12 {class=MaterialExpressionNamedRerouteUsage}
  MaterialExpressionNamedRerouteUsage_15 {class=MaterialExpressionNamedRerouteUsage}
  MaterialExpressionNamedRerouteUsage_2 {class=MaterialExpressionNamedRerouteUsage}
  MaterialExpressionNamedRerouteUsage_3 {class=MaterialExpressionNamedRerouteUsage}
  MaterialExpressionNamedRerouteUsage_4 {class=MaterialExpressionNamedRerouteUsage}
  MaterialExpressionNamedRerouteUsage_5 {class=MaterialExpressionNamedRerouteUsage}
  MaterialExpressionNamedRerouteUsage_6 {class=MaterialExpressionNamedRerouteUsage}
  MaterialExpressionNamedRerouteUsage_7 {class=MaterialExpressionNamedRerouteUsage}
  MaterialExpressionSaturate_0 {class=MaterialExpressionSaturate, inputs [None {from=MaterialExpressionSubtract_4}]}
  MaterialExpressionSaturate_4 {class=MaterialExpressionSaturate, inputs [None {from=MaterialExpressionConvert_0}]}
  MaterialExpressionScalarParameter_0 {class=MaterialExpressionScalarParameter, properties {group=RankIcon, parameterName=IconNormalFlatness}}
  MaterialExpressionScalarParameter_1 {class=MaterialExpressionScalarParameter, properties {defaultValue=0.23, group=RankIcon, parameterName=IconUVScale}}
  MaterialExpressionScalarParameter_10 {class=MaterialExpressionScalarParameter, properties {defaultValue=-0.328, group=RankIcon, parameterName=IconMirrorXOffset}}
  MaterialExpressionScalarParameter_11 {class=MaterialExpressionScalarParameter, properties {defaultValue=-0.03, group=RankIcon, parameterName=IconMirrorYOffset}}
  MaterialExpressionScalarParameter_12 {class=MaterialExpressionScalarParameter, properties {group=RankIcon, parameterName=AddRemoveMask_Blue, sliderMax=1, sortPriority=5}}
  MaterialExpressionScalarParameter_13 {
   class=MaterialExpressionScalarParameter
   properties {defaultValue=1, group=RankIcon, parameterName=AddRemoveMask_Green, sliderMax=1, sortPriority=4}
  }
  MaterialExpressionScalarParameter_14 {
   class=MaterialExpressionScalarParameter
   properties {defaultValue=1, group=RankIcon, parameterName=AddRemoveMask_Red, sliderMax=1, sortPriority=3}
  }
  MaterialExpressionScalarParameter_18 {class=MaterialExpressionScalarParameter, properties {group=RankIcon, parameterName=RankIndex}}
  MaterialExpressionScalarParameter_2 {class=MaterialExpressionScalarParameter, properties {defaultValue=0.42, group=RankIcon, parameterName=SphereMaskX}}
  MaterialExpressionScalarParameter_3 {class=MaterialExpressionScalarParameter, properties {defaultValue=-0.96, group=RankIcon, parameterName=IconUVOffsetX}}
  MaterialExpressionScalarParameter_4 {class=MaterialExpressionScalarParameter, properties {defaultValue=0.049, group=RankIcon, parameterName=IconUVOffsetY}}
  MaterialExpressionScalarParameter_5 {class=MaterialExpressionScalarParameter, properties {defaultValue=0.41, group=RankIcon, parameterName=SphereMaskY}}
  MaterialExpressionScalarParameter_6 {class=MaterialExpressionScalarParameter, properties {defaultValue=0.015, group=RankIcon, parameterName=IconUVRotation}}
  MaterialExpressionScalarParameter_7 {class=MaterialExpressionScalarParameter, properties {defaultValue=0.6, group=RankIcon, parameterName=IconUVMirror}}
  MaterialExpressionScalarParameter_8 {class=MaterialExpressionScalarParameter, properties {defaultValue=0.05, group=RankIcon, parameterName=SphereMaskBoundsX}}
  MaterialExpressionScalarParameter_9 {class=MaterialExpressionScalarParameter, properties {defaultValue=0.05, group=RankIcon, parameterName=SphereMaskBoundsY}}
  MaterialExpressionSetMaterialAttributes_0 {
   class=MaterialExpressionSetMaterialAttributes
   inputs [
    MaterialAttributes {from=MaterialExpressionGetMaterialAttributes_1, output=MaterialAttributes}
    BaseColor {from=MaterialExpressionNamedRerouteUsage_7, output=""}
   ]
   properties {attributeSetTypes [#2]}
  }
  MaterialExpressionSetMaterialAttributes_1 {
   class=MaterialExpressionSetMaterialAttributes
   inputs [
    MaterialAttributes {from=MaterialExpressionGetMaterialAttributes_1, output=MaterialAttributes}
    Normal {from=MaterialExpressionNamedRerouteUsage_2, output=""}
   ]
   properties {attributeSetTypes [#1]}
  }
  MaterialExpressionStaticBoolParameter_0 {class=MaterialExpressionStaticBoolParameter, inputs [], properties {group=RankIcon, parameterName=DeactivateRankedIcon, sortPriority=0}}
  MaterialExpressionStaticSwitchParameter_0 {
   class=MaterialExpressionStaticSwitchParameter
   inputs [
    True {from=MaterialExpressionSetMaterialAttributes_0, output=""}
    False {from=MaterialExpressionBlendMaterialAttributes_1, output=""}
   ]
   properties {group=RankIcon, parameterName=DEBUG-IconBoxMask}
  }
  MaterialExpressionStaticSwitch_0 {
   class=MaterialExpressionStaticSwitch
   inputs [
    True {from=MaterialExpressionFunctionInput_1, output=""}
    False {from=MaterialExpressionStaticSwitchParameter_0, output=""}
    Value {from=MaterialExpressionStaticBoolParameter_0, output=""}
   ]
   properties {}
  }
  MaterialExpressionStaticSwitch_1 {
   class=MaterialExpressionStaticSwitch
   inputs [
    True {from=MaterialExpressionConstant_0, output=""}
    False {from=MaterialExpressionMultiply_6, output=""}
    Value {from=MaterialExpressionStaticBoolParameter_0, output=""}
   ]
   properties {}
  }
  MaterialExpressionStep_0 {
   class=MaterialExpressionStep
   inputs [
    Y {from="", output=""}
    X {from=MaterialExpressionMaterialFunctionCall_5, output=Result}
   ]
   properties {constY=0.1}
  }
  MaterialExpressionStep_1 {
   class=MaterialExpressionStep
   inputs [
    Y {from=MaterialExpressionMaterialFunctionCall_6, output=UGradient}
    X {from=MaterialExpressionScalarParameter_7, output=""}
   ]
   properties {constX=0.3}
  }
  MaterialExpressionSubtract_2 {
   class=MaterialExpressionSubtract
   inputs [
    A {from=MaterialExpressionConvert_2}
    B {from=MaterialExpressionNamedRerouteUsage_11}
   ]
  }
  MaterialExpressionSubtract_3 {
   class=MaterialExpressionSubtract
   inputs [
    A {from=MaterialExpressionConvert_2}
    B {from=MaterialExpressionNamedRerouteUsage_11}
   ]
  }
  MaterialExpressionSubtract_4 {
   class=MaterialExpressionSubtract
   inputs [
    A {from=MaterialExpressionNamedRerouteUsage_15}
    B {from=MaterialExpressionStep_0}
   ]
  }
  MaterialExpressionSubtract_5 {
   class=MaterialExpressionSubtract
   inputs [
    A {from=MaterialExpressionNamedRerouteDeclaration_8}
    B {from=MaterialExpressionNamedRerouteUsage_11}
   ]
  }
  MaterialExpressionTextureCoordinate_0 {class=MaterialExpressionTextureCoordinate}
  MaterialExpressionTextureCoordinate_1 {class=MaterialExpressionTextureCoordinate}
  MaterialExpressionTextureSample_0 {
   class=MaterialExpressionTextureSample
   inputs [
    UVs {from=MaterialExpressionNamedRerouteUsage_6}
    Tex {from=""}
    Apply View MipBias {from=""}
   ]
   properties {samplerType=SAMPLERTYPE_Normal, texture=/Game/Global/Materials/Cosmetics/Ranked/Textures/T_Ranked_Icon3x3_N}
  }
  MaterialExpressionTextureSample_2 {
   class=MaterialExpressionTextureSample
   inputs [
    UVs {from=MaterialExpressionNamedRerouteUsage_5}
    Tex {from=""}
    Apply View MipBias {from=""}
   ]
   properties {samplerType=SAMPLERTYPE_Masks, texture=/Game/Global/Materials/Cosmetics/Ranked/Textures/T_Ranked_Icon3x3}
  }
 }}
