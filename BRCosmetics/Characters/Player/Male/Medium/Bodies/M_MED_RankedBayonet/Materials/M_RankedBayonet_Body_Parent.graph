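# Parent material for the RankedBayonet male body cosmetic, layering ranked effects between the shared character PreFX and PostFX functions.
# It chains icon, thin-film and emissive Fresnel functions, mostly borrowed from the female RankedScythe body.
# An optional InteriorCoatGlow switch fakes a light behind the character using pre-skinned position, normal and camera vector.
# That two-color glow, reversed through a tiled triangle pattern, blends in through a TechArt mask channel.
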Material /BRCosmetics/Characters/Player/Male/Medium/Bodies/M_MED_RankedBayonet/Materials/M_RankedBayonet_Body_Parent
graph {
 properties {
  bUseMaterialAttributes=true
  bUsedWithMeshDeformer=true
  bUsedWithMorphTargets=true
  bUsedWithNanite=true
  bUsedWithSkeletalMesh=true
  pixelDepthOffsetMode=PDOM_Legacy
  shadingModel=MSM_FromMaterialExpression
  usedShadingModels=Default Lit | Subsurface | Clear Coat | Subsurface Profile | Cloth
 }
 nodes [
  n1 = MaterialFunctionCall() {materialFunction=&MF_CharacterParent_2023_PreFX}
  MaterialFunctionCall(Material=NamedRerouteUsage(), M=@n1.M, Pre Skinned Local Position=@n1.M, Distance Blend=@n1.M, Fresnel=@n1.M) {materialFunction=&MF_CharacterParent_2023_PostFX}
  n2 = MaterialFunctionCall!(Material=@n1.Result, TechArt Mask=NamedRerouteDeclaration(None=@n1.Tech Art Mask) {name=TA Mask, nodeColor=<rgba 0.125 0 1 1>}) {materialFunction=/BRCosmetics/Characters/Player/Female/Medium/Bodies/F_MED_RankedScythe/Materials/MF_RankedScytheIcon}
  NamedRerouteDeclaration(None=MaterialFunctionCall!(
    Material=MaterialFunctionCall!(Material Attributes=@n2.Result, Alpha=@n2.Result) {materialFunction=/Game/Global/Materials/Cosmetics/Ranked/MF_RankedThinFilm}.Result
    Mask=@n2.TechArt Mask
    Fresnel=@n1.Fresnel
   ) {materialFunction=/BRCosmetics/Characters/Player/Female/Medium/Bodies/F_MED_RankedScythe/Materials/MF_RankedScythe_EmissiveFresnel}.Result) {name=MA Original, nodeColor=<rgba 0 1 0.7656 1>}
  n3 = NamedRerouteUsage()
  n4 = VertexNormalWS()
  n5 = ScalarParameter() {defaultValue=200, group=RankedScytheGlow, parameterName=InteriorGlowNormalContribution}
  n6 = VertexInterpolator(VS=LocalPosition() {localOrigin=InstancePreSkinning}.XYZ)
  n7 = CameraVectorWS()
  n8 = Add(
   A=VectorParameter() {group=RankedScytheGlow, parameterName=InteriorGlowLocation, defaultValue=<rgba 0 0 200 1>}.RGB
   B=Multiply(A=@n7, B=ScalarParameter() {defaultValue=-0.7177, group=RankedScytheGlow, parameterName=InteriorDistanceBehindCharacter})
  )
  n9 = Normalize(VectorInput=Add(A=Multiply(A=@n4, B=@n5), B=Subtract(A=@n6.PS, B=@n8)))
  n10 = Saturate(None=Power(Base=DotProduct(A=@n9, B=@n7), Exp=ScalarParameter() {defaultValue=4.028, group=RankedScytheGlow, parameterName=FalloffInteriorglow}))
  n11 = LinearInterpolate(
   A=VectorParameter() {group=RankedScytheGlow, parameterName=ColorA, defaultValue=<rgba 0.3477 0.856 0.89 1>}.RGB
   B=VectorParameter() {group=RankedScytheGlow, parameterName=ColorB, defaultValue=<rgba 0.0931 0.3663 0.624 1>}.RGB
   Alpha=@n10
  )
  n12 = TextureSample(UVs=Multiply(A=TextureCoordinate(), B=ScalarParameter() {defaultValue=32, group=RankedScytheGlow, parameterName=PatternTilling})) {
   samplerType=SAMPLERTYPE_LinearGrayscale
   texture=/BRCosmetics/Characters/Player/Female/Medium/Bodies/F_MED_RankedScythe/Textures/T_TrianglePattern_mask
  }
  n13 = LinearInterpolate(
   A=LinearInterpolate(
    A=VectorParameter() {group=RankedScytheGlow, parameterName=ColorB, defaultValue=<rgba 0.0931 0.3663 0.624 1>}.RGB
    B=VectorParameter() {group=RankedScytheGlow, parameterName=ColorA, defaultValue=<rgba 0.3477 0.856 0.89 1>}.RGB
    Alpha=@n10
   )
   B=@n11
   Alpha=NamedRerouteDeclaration(None=@n12.R) {name=Pattern, nodeColor=<rgba 0.6406 1 0 1>}
  )
  n14 = SetMaterialAttributes!(MaterialAttributes=GetMaterialAttributes(None=@n3).MaterialAttributes, Base Color=@n11, Emissive Color=Multiply(A=@n13, B=@n13)) {attributeSetTypes [#1, #2]}
  n15 = ChannelMaskParameter(None=NamedRerouteUsage()) {group=RankedScytheGlow, maskChannel=Green, parameterName=TriangleMask_Channel, defaultValue=<rgba 0 1 0 0>}
  NamedRerouteDeclaration(None=StaticSwitchParameter(True=BlendMaterialAttributes(A=@n3, B=@n14, Alpha=@n15.R), False=@n3) {group=RankedScytheGlow, parameterName=InteriorCoatGlow, sortPriority=0}) {name=MA Final, nodeColor=<rgba 1 0.4453 0 1>}
 ]
}
