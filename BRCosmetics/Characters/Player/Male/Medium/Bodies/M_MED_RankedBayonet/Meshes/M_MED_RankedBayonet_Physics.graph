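# Physics asset for the RankedBayonet male medium body cosmetic, covering six jacket panels (front, back and side pairs) and a dangling bibelot.
# Pelvis and spine capsules form a locked chain, with jacket base bodies locked to spine_05 and the bibelot base to pelvis.
# Each dynamic tip hangs from its base with tight twist and per-panel swing limits.
# A left thigh capsule is free-jointed to the pelvis, likely as a collider.
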PhysicsAsset /BRCosmetics/Characters/Player/Male/Medium/Bodies/M_MED_RankedBayonet/Meshes/M_MED_RankedBayonet_Physics
bodies {
 dyn_bibelot [{shapeType=Capsule, length=3.1906, radius=2.9093, shapeName=dyn_bibelot_capsule, center=<2.5488 -0.365 0.2181>}]
 dyn_bibelot_base [{shapeType=Capsule, length=0.505, radius=0.505, shapeName=dyn_bibelot_base_capsule, rotation=<rot 0 -90 0>}]
 dyn_jacket_bk_l [{shapeType=Capsule, length=5, radius=5, shapeName=dyn_jacket_bk_l_capsule, center=<5 0 0>, rotation=<rot 0 0 90>}]
 dyn_jacket_bk_l_base [{shapeType=Capsule, length=1, radius=1, shapeName=dyn_jacket_bk_l_base_capsule}]
 dyn_jacket_bk_r [{shapeType=Capsule, length=5, radius=5, shapeName=dyn_jacket_bk_r_capsule, center=<5 0 0>, rotation=<rot 0 0 90>}]
 dyn_jacket_bk_r_base [{shapeType=Capsule, length=1, radius=1, shapeName=dyn_jacket_bk_r_base_capsule}]
 dyn_jacket_fr_l [{shapeType=Capsule, length=5, radius=5, shapeName=dyn_jacket_fr_l_capsule, center=<5 0 0>, rotation=<rot 0 0 90>}]
 dyn_jacket_fr_l_base [{shapeType=Capsule, length=1, radius=1, shapeName=dyn_jacket_fr_l_base_capsule}]
 dyn_jacket_fr_r [{shapeType=Capsule, length=5, radius=5, shapeName=dyn_jacket_fr_r_capsule, center=<5 0 0>, rotation=<rot 0 0 90>}]
 dyn_jacket_fr_r_base [{shapeType=Capsule, length=1, radius=1, shapeName=dyn_jacket_fr_r_base_capsule}]
 dyn_jacket_side_l [{shapeType=Capsule, length=5, radius=5, shapeName=dyn_jacket_side_l_capsule, center=<5 0 0>, rotation=<rot 0 0 90>}]
 dyn_jacket_side_l_base [{shapeType=Capsule, length=1, radius=1, shapeName=dyn_jacket_side_l_base_capsule}]
 dyn_jacket_side_r [{shapeType=Capsule, length=5, radius=5, shapeName=dyn_jacket_side_r_capsule, center=<5 0 0>, rotation=<rot 0 0 90>}]
 dyn_jacket_side_r_base [{shapeType=Capsule, length=1, radius=1, shapeName=dyn_jacket_side_r_base_capsule}]
 pelvis [
  {shapeType=Capsule, length=12, radius=8, center=<0 -1.5991 0>}
  {shapeType=Capsule, length=12, radius=8, shapeName=pelvis_capsule, center=<6.2258 -0.7446 0>}
 ]
 spine_03 [{shapeType=Capsule, length=7, radius=7}]
 spine_04 [{shapeType=Capsule, length=7, radius=7}]
 spine_05 [{shapeType=Capsule, length=7, radius=7}]
 thigh_l [{shapeType=Capsule, length=15, radius=5, shapeName=thigh_l_capsule, center=<-15 0 0>, rotation=<rot 90 0 0>}]
}
constraints [
 {
  bone1Name=spine_03
  bone2Name=pelvis
  swing1LimitDegrees=45
  swing1Motion=Locked
  swing2LimitDegrees=45
  swing2Motion=Locked
  twistLimitDegrees=45
  twistMotion=Locked
 }
 {
  bone1Name=spine_04
  bone2Name=spine_03
  swing1LimitDegrees=45
  swing1Motion=Locked
  swing2LimitDegrees=45
  swing2Motion=Locked
  twistLimitDegrees=45
  twistMotion=Locked
 }
 {
  bone1Name=spine_05
  bone2Name=spine_04
  swing1LimitDegrees=45
  swing1Motion=Locked
  swing2LimitDegrees=45
  swing2Motion=Locked
  twistLimitDegrees=45
  twistMotion=Locked
 }
 {
  bone1Name=dyn_jacket_bk_l_base
  bone2Name=spine_05
  swing1LimitDegrees=45
  swing1Motion=Locked
  swing2LimitDegrees=45
  swing2Motion=Locked
  twistLimitDegrees=45
  twistMotion=Locked
 }
 {bone1Name=dyn_jacket_bk_l, bone2Name=dyn_jacket_bk_l_base, swing1LimitDegrees=15, swing2LimitDegrees=45, twistLimitDegrees=5}
 {
  bone1Name=dyn_jacket_bk_r_base
  bone2Name=spine_05
  swing1LimitDegrees=45
  swing1Motion=Locked
  swing2LimitDegrees=45
  swing2Motion=Locked
  twistLimitDegrees=45
  twistMotion=Locked
 }
 {bone1Name=dyn_jacket_bk_r, bone2Name=dyn_jacket_bk_r_base, swing1LimitDegrees=15, swing2LimitDegrees=45, twistLimitDegrees=5}
 {
  bone1Name=dyn_jacket_fr_l_base
  bone2Name=spine_05
  swing1LimitDegrees=45
  swing1Motion=Locked
  swing2LimitDegrees=45
  swing2Motion=Locked
  twistLimitDegrees=45
  twistMotion=Locked
 }
 {bone1Name=dyn_jacket_fr_l, bone2Name=dyn_jacket_fr_l_base, swing1LimitDegrees=30, swing2LimitDegrees=35, twistLimitDegrees=5}
 {
  bone1Name=dyn_jacket_fr_r_base
  bone2Name=spine_05
  swing1LimitDegrees=45
  swing1Motion=Locked
  swing2LimitDegrees=45
  swing2Motion=Locked
  twistLimitDegrees=45
  twistMotion=Locked
 }
 {bone1Name=dyn_jacket_fr_r, bone2Name=dyn_jacket_fr_r_base, swing1LimitDegrees=30, swing2LimitDegrees=35, twistLimitDegrees=5}
 {
  bone1Name=dyn_jacket_side_l_base
  bone2Name=spine_05
  swing1LimitDegrees=45
  swing1Motion=Locked
  swing2LimitDegrees=45
  swing2Motion=Locked
  twistLimitDegrees=45
  twistMotion=Locked
 }
 {bone1Name=dyn_jacket_side_l, bone2Name=dyn_jacket_side_l_base, swing1LimitDegrees=10, swing2LimitDegrees=25, twistLimitDegrees=5}
 {
  bone1Name=dyn_jacket_side_r_base
  bone2Name=spine_05
  swing1LimitDegrees=45
  swing1Motion=Locked
  swing2LimitDegrees=45
  swing2Motion=Locked
  twistLimitDegrees=45
  twistMotion=Locked
 }
 {bone1Name=dyn_jacket_side_r, bone2Name=dyn_jacket_side_r_base, swing1LimitDegrees=10, swing2LimitDegrees=25, twistLimitDegrees=5}
 {
  bone1Name=dyn_bibelot_base
  bone2Name=pelvis
  swing1LimitDegrees=45
  swing1Motion=Locked
  swing2LimitDegrees=45
  swing2Motion=Locked
  twistLimitDegrees=45
  twistMotion=Locked
 }
 {bone1Name=dyn_bibelot, bone2Name=dyn_bibelot_base, swing1LimitDegrees=45, swing2LimitDegrees=10, twistLimitDegrees=5}
 {
  bone1Name=thigh_l
  bone2Name=pelvis
  swing1LimitDegrees=45
  swing1Motion=Free
  swing2LimitDegrees=45
  swing2Motion=Free
  twistLimitDegrees=45
  twistMotion=Free
 }
]
