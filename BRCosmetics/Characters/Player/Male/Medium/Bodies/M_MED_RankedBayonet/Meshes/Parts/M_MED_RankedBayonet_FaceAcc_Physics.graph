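# Physics asset for the face-accessory mesh of the male medium RankedBayonet player body, defining capsule bodies and joints for its hood.
# Six dyn_hood bones swing and twist within limits on hood_correction_base bones, which are locked to neck_02 (rear pair) or spine_05 (front and side pairs).
# The neck_02 to spine_05 joint is free.
# The dyn_hood_l/r bones attach crosswise to opposite-side bases, unlike the front and side pairs.
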PhysicsAsset /BRCosmetics/Characters/Player/Male/Medium/Bodies/M_MED_RankedBayonet/Meshes/Parts/M_MED_RankedBayonet_FaceAcc_Physics
bodies {
 dyn_hood_fr_l [{shapeType=Capsule, length=1, radius=2, shapeName=dyn_hood_fr_l_capsule, center=<5 0 0>}]
 dyn_hood_fr_r [{shapeType=Capsule, length=1, radius=2, shapeName=dyn_hood_fr_r_capsule, center=<5 0 0>}]
 dyn_hood_l [{shapeType=Capsule, length=6, radius=2, shapeName=dyn_hood_l_capsule, center=<6 0 0>, rotation=<rot 90 0 0>}]
 dyn_hood_r [{shapeType=Capsule, length=6, radius=2, shapeName=dyn_hood_r_capsule, center=<6 0 0>, rotation=<rot 90 0 0>}]
 dyn_hood_side_l [{shapeType=Capsule, length=6, radius=2, shapeName=dyn_hood_side_l_capsule, center=<6 0 0>, rotation=<rot 90 0 0>}]
 dyn_hood_side_r [{shapeType=Capsule, length=6, radius=2, shapeName=dyn_hood_side_r_capsule, center=<6 0 0>, rotation=<rot 90 0 0>}]
 hood_correction_base_fr_l [{shapeType=Capsule, length=1, radius=2, shapeName=hood_correction_base_fr_l_capsule}]
 hood_correction_base_fr_r [{shapeType=Capsule, length=1, radius=2, shapeName=hood_correction_base_fr_r_capsule}]
 hood_correction_base_l [{shapeType=Capsule, length=1, radius=1, shapeName=hood_correction_base_l_capsule}]
 hood_correction_base_r [{shapeType=Capsule, length=1, radius=1, shapeName=hood_correction_base_r_capsule}]
 hood_correction_base_side_l [{shapeType=Capsule, length=1, radius=1, shapeName=hood_correction_base_side_l_capsule}]
 hood_correction_base_side_r [{shapeType=Capsule, length=1, radius=1, shapeName=hood_correction_base_side_r_capsule}]
 neck_02 [{shapeType=Capsule, length=0.505, radius=0.505, shapeName=neck_02_capsule, rotation=<rot 0 -90 0>}]
 spine_05 [{shapeType=Capsule, length=1, radius=2, shapeName=spine_05_capsule}]
}
constraints [
 {
  bone1Name=hood_correction_base_l
  bone2Name=neck_02
  swing1LimitDegrees=45
  swing1Motion=Locked
  swing2LimitDegrees=45
  swing2Motion=Locked
  twistLimitDegrees=45
  twistMotion=Locked
 }
 {bone1Name=dyn_hood_r, bone2Name=hood_correction_base_l, swing1LimitDegrees=15, swing2LimitDegrees=45, twistLimitDegrees=5}
 {
  bone1Name=hood_correction_base_r
  bone2Name=neck_02
  swing1LimitDegrees=45
  swing1Motion=Locked
  swing2LimitDegrees=45
  swing2Motion=Locked
  twistLimitDegrees=45
  twistMotion=Locked
 }
 {bone1Name=dyn_hood_l, bone2Name=hood_correction_base_r, swing1LimitDegrees=15, swing2LimitDegrees=45, twistLimitDegrees=5}
 {bone1Name=dyn_hood_fr_l, bone2Name=hood_correction_base_fr_l, swing1LimitDegrees=15, swing2LimitDegrees=45, twistLimitDegrees=5}
 {bone1Name=dyn_hood_fr_r, bone2Name=hood_correction_base_fr_r, swing1LimitDegrees=15, swing2LimitDegrees=45, twistLimitDegrees=5}
 {bone1Name=dyn_hood_side_l, bone2Name=hood_correction_base_side_l, swing1LimitDegrees=15, swing2LimitDegrees=45, twistLimitDegrees=5}
 {bone1Name=dyn_hood_side_r, bone2Name=hood_correction_base_side_r, swing1LimitDegrees=15, swing2LimitDegrees=45, twistLimitDegrees=5}
 {
  bone1Name=neck_02
  bone2Name=spine_05
  swing1LimitDegrees=45
  swing1Motion=Free
  swing2LimitDegrees=45
  swing2Motion=Free
  twistLimitDegrees=45
  twistMotion=Free
 }
 {
  bone1Name=hood_correction_base_side_l
  bone2Name=spine_05
  swing1LimitDegrees=45
  swing1Motion=Locked
  swing2LimitDegrees=45
  swing2Motion=Locked
  twistLimitDegrees=45
  twistMotion=Locked
 }
 {
  bone1Name=hood_correction_base_side_r
  bone2Name=spine_05
  swing1LimitDegrees=45
  swing1Motion=Locked
  swing2LimitDegrees=45
  swing2Motion=Locked
  twistLimitDegrees=45
  twistMotion=Locked
 }
 {
  bone1Name=hood_correction_base_fr_l
  bone2Name=spine_05
  swing1LimitDegrees=45
  swing1Motion=Locked
  swing2LimitDegrees=45
  swing2Motion=Locked
  twistLimitDegrees=45
  twistMotion=Locked
 }
 {
  bone1Name=hood_correction_base_fr_r
  bone2Name=spine_05
  swing1LimitDegrees=45
  swing1Motion=Locked
  swing2LimitDegrees=45
  swing2Motion=Locked
  twistLimitDegrees=45
  twistMotion=Locked
 }
]
